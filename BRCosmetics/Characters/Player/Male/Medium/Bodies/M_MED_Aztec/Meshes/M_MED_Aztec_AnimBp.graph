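# Anim Blueprint for the M_MED_Aztec player body, adding secondary motion to jacket bones on a pose copied from the parent mesh.
# The zipper bone gets a constraint and AnimDynamics with a thigh collision sphere, then feeds a cached pose into two jacket-back simulations.
# Skydiving or wearing a cape selects the jacket-back variant with zeroed angular limits.
# Cape and backbling flags are set once when character customization loads, using CheckForBackbling.
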AnimBlueprint /BRCosmetics/Characters/Player/Male/Medium/Bodies/M_MED_Aztec/Meshes/M_MED_Aztec_AnimBp
target_skeleton=&/M_MED_Aztec_Skeleton
graphs [
 AnimGraph {nodes [
   n1 = Anim Dynamics - Bone: dyn_jacket_back(ComponentPose=Local To Component(LocalPose=Use cached pose 'PreJacketDyn'()), Alpha=1, ExternalForce="0,0,0") {
    angularDampingOverride=0.98
    angularSpringConstant=500
    bAngularSpring=true
    bGravityOverrideInSimSpace=true
    bLinearSpring=true
    gravityScale=0.25
    linearDampingOverride=0.98
    linearSpringConstant=500
    simulationSpace=RootRelative
    alphaBoolBlend {blendOption=Linear}
    alphaScaleBiasClamp {bClampResult=true, bInterpResult=true, clampMin=0.2, interpSpeedDecreasing=5, interpSpeedIncreasing=5}
    boundBone {boneName=dyn_jacket_back}
    componentAppliedLinearAccClamp=<900 900 900>
    componentLinearAccScale=<0.1 0.1 0.1>
    componentLinearVelScale=<0.5 0.5 0.5>
    physicsBodyDefinitions [{
      sphereCollisionRadius=0
      boundBone {boneName=dyn_jacket_back}
      boxExtents=<20 5 20>
      constraintSetup {twistAxis=AxisY, angularLimitsMax=<80 0 0>, angularLimitsMin=<-2 0 0>, angularTarget=<0 0 0>}
      localJointOffset=<0 -2 5>
     }]
    planarLimits [{drivingBone {boneName=pelvis}, planeTransform {location=<0 -15 0>, rotation=<rot 0 0 -89>}}]
    relativeSpaceBone {boneName=root}
   }
   n2 = Anim Dynamics - Bone: dyn_jacket_back(ComponentPose=Local To Component(LocalPose=Use cached pose 'PreJacketDyn'()), Alpha=1, ExternalForce="0,0,0") {
    angularDampingOverride=0.9999
    bAngularSpring=true
    bGravityOverrideInSimSpace=true
    bLinearSpring=true
    gravityScale=0.25
    linearDampingOverride=0.9999
    simulationSpace=RootRelative
    alphaBoolBlend {blendOption=Linear}
    alphaScaleBiasClamp {bClampResult=true, bInterpResult=true, clampMin=0.2, interpSpeedDecreasing=5, interpSpeedIncreasing=5}
    boundBone {boneName=dyn_jacket_back}
    componentAppliedLinearAccClamp=<900 900 900>
    componentLinearAccScale=<0.1 0.1 0.1>
    componentLinearVelScale=<0.5 0.5 0.5>
    physicsBodyDefinitions [{
      sphereCollisionRadius=0
      boundBone {boneName=dyn_jacket_back}
      boxExtents=<20 5 20>
      constraintSetup {twistAxis=AxisY, angularLimitsMax=<0 0 0>, angularLimitsMin=<0 0 0>, angularTarget=<0 0 0>}
      localJointOffset=<0 -2 5>
     }]
    planarLimits [{drivingBone {boneName=pelvis}, planeTransform {location=<0 -15 0>, rotation=<rot 0 0 -89>}}]
    relativeSpaceBone {boneName=root}
   }
   n3 = Get isWearingCape() -> (isWearingCape:bool)
   n4 = Get bIsSkyDiving()
   Output Pose(Result=Blend Poses by bool(BlendPose_0=Component To Local(ComponentPose=@n2), BlendPose_1=Component To Local(ComponentPose=@n1), bActiveValue=OR Boolean(A=@n4, B=@n3)) {blendType=Linear}) {layerGroup=None}
   n5 = "Constraint - dyn_jacketzipper (PPPP)"!(
    ComponentPose=Local To Component(LocalPose=Copy Pose From Mesh() {bUseAttachedParent=true})
    Alpha=1
    ConstraintWeights_0=0.0386
    ConstraintWeights_1:float=0.0011
    ConstraintWeights_2:float=0.4481
    ConstraintWeights_3:float=0.5122
   ) -> (Pose) {
    alphaBoolBlend {blendOption=Linear}
    boneToModify {boneName=dyn_jacketzipper}
    constraintSetup [
     {targetBone {boneName=spine_01}}
     {targetBone {boneName=thigh_twist_01_l}}
     {targetBone {boneName=thigh_twist_01_r}}
     {targetBone {boneName=pelvis}}
    ]
    constraintWeights [1,1,1,1]
   }
   n6 = Anim Dynamics - Bone: dyn_jacketzipper(ComponentPose=@n5, Alpha=1, ExternalForce="0,0,0") {
    angularDampingOverride=0.96
    bGravityOverrideInSimSpace=true
    bOverrideAngularDamping=true
    bOverrideLinearDamping=true
    bUseSphericalLimits=true
    gravityScale=0.25
    linearDampingOverride=0.96
    simulationSpace=Actor
    alphaBoolBlend {blendOption=Linear}
    boundBone {boneName=dyn_jacketzipper}
    componentAppliedLinearAccClamp=<700 700 700>
    componentLinearAccScale=<0.15 0.15 0.15>
    componentLinearVelScale=<0.15 0.15 0.15>
    physicsBodyDefinitions [{
      sphereCollisionRadius=0
      boundBone {boneName=dyn_jacketzipper}
      boxExtents=<2 8 4>
      constraintSetup {angularLimitsMax=<0 0 30>, angularLimitsMin=<0 0 -70>, angularTarget=<0 0 0>}
      localJointOffset=<0 -2 0>
     }]
    sphericalLimits [{limitRadius=4, drivingBone {boneName=thigh_r}, sphereLocalOffset=<6 -9 1>}]
   }
   Save cached pose 'PreJacketDyn'(Pose=Component To Local(ComponentPose=@n6)) {cacheName=PreJacketDyn}
  ]}
 EventGraph {nodes [
   Event OnCharacterCustomizationLoaded()
   > Do Once()
   n1 = CheckForBackbling()
   > Set isWearingBackbling(isWearingBackbling=OR Boolean(A=@n1, B=@n1.bIsBackblingLong))
   n2 = GetPartSkeletalMeshComponent()
   n3 = FindChildCommponentFromAttachSocket!(
    self:Custom Character Part Anim Instance Object Reference
    Parent:Skeletal Mesh Component Object Reference=@n2
    bIncludeAllDescendants:bool=false
    AttachSocket:Name=Cape
   ) -> (ReturnValue:Skeletal Mesh Component Object Reference)
   > Branch(Condition="Not Equal (Object)"!(A:Object Reference=@n3, B:Object Reference) -> (ReturnValue:bool))
   > Set isWearingCape(isWearingCape=@n1.bIsBackblingLong)
   "Not Equal (Object)"(A=FindChildCommponentFromAttachSocket(Parent=@n2, bIncludeAllDescendants=false, AttachSocket=backpack_br))
   Event BlueprintUpdateAnimation()
   Reroute Node(InputPin:bool="float > float"(
     A=Get Distance Between Two Sockets!(
      Component:Skeletal Mesh Component Object Reference=@n3
      SocketOrBoneNameA:Name=backpack_root
      SocketSpaceA:ERelativeTransformSpace Enum=RTS_World
      SocketOrBoneNameB:Name=cape_length
      SocketSpaceB:ERelativeTransformSpace Enum=RTS_World
      bRemapRange:bool=false
      InRangeMin:float=0
      InRangeMax:float=0
      OutRangeMin:float=0
      OutRangeMax:float=0
     ) -> (ReturnValue:float) {is=title K2Node_CallFunction}
     B=50
    ) {is=title K2Node_CallFunction}) -> (OutputPin:bool)
  ]}
]
variables [
 isWearingBackbling {type=boolean}
 isWearingCape {type=boolean}
]
